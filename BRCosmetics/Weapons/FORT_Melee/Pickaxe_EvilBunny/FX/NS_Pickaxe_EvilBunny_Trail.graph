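# Niagara system for the swing trail of the EvilBunny pickaxe cosmetic, drawn as two ribbon emitters built on shared anim-trail modules.
# Main_Trail is a dim blue-grey ribbon, while Tip_Trail is a thin pale ribbon measured from one socket that tapers over particle life.
# Both are short-lived, darken and fade through curves, and scale alpha by time of day.
# Main_Trail also multiplies color by a system time-of-day multiplier.
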NiagaraSystem /BRCosmetics/Weapons/FORT_Melee/Pickaxe_EvilBunny/FX/NS_Pickaxe_EvilBunny_Trail
dependencies {
 usedDataInterfaces [&NiagaraDataInterfaceColorCurve, &NiagaraDataInterfaceCurve, &NiagaraDataInterfaceSocketReader]
 usedDynamicInputs []
 usedModules [
  &self:SystemSpawnScript.SystemScriptSource.SystemScriptGraph.NiagaraNodeAssignment_1.SetVariables_#1
  &SystemState
  &self:SystemSpawnScript.SystemScriptSource.SystemScriptGraph.NiagaraNodeAssignment_0.SetVariables_#2
 ]
}
emitters {
 Main_Trail {
  emitterSpawnScript [&FN_RibbonIDOffsetUtility()]
  emitterUpdateScript [
   &EmitterState()
   &FN_SpawnAnimTrail()
  ]
  particleSpawnScript [
   &FN_InitializeAnimTrailBoneIndecies()
   &FN_InitializeAnimTrail(Color=<rgba 0.1332 0.1676 0.2664 0.777>, Lifetime=0.3)
  ]
  particleUpdateScript [
   &SolveForcesAndVelocity(Clamp Velocity=false, Limit Acceleration=false)
   &FN_UpdateAnimTrail() {module=FN_UpdateAnimTrail001}
   &ScaleColor(
    Scale Alpha=&FloatFromCurve(
     FloatCurve=DataInterface() {propertyValues {Curve {keys [
         {arriveTangent=-2.6867, interpMode=RCIM_Cubic, time=0, value=1}
         {arriveTangent=-6.5457, interpMode=RCIM_Cubic, leaveTangent=-6.5457, tangentMode=RCTM_User, time=0.5, value=0}
        ]}}}
     Scale Curve=&FNTimeOfDayScalar(Night Scale=1.5, Morning Scale=1, Evening Scale=1)
    )
    Scale RGB=&MakeVectorFromLinearColorRGB(Color=&ColorFromCurve(ColorCurve=DataInterface() {propertyValues {
        AlphaCurve {keys [<at 0 1 RCIM_Cubic>,<at 1 0>]}
        BlueCurve {keys [<at 0 1 RCIM_Cubic>,<at 1 0.05 RCIM_Cubic>]}
        GreenCurve {keys [<at 0 1 RCIM_Cubic>,<at 1 0.05 RCIM_Cubic>]}
        RedCurve {keys [<at 0 1 RCIM_Cubic>,<at 1 0.05 RCIM_Cubic>]}
       }}))
    ScaleA=true
    ScaleRGB=true
   )
   /SetVariables(Particles.Color=&MultiplyLinearColor(
     Linear Color=$Particles.Color
     Scale Factor=&MakeVector4_FromLinearColor(LinearColor=$System.TODMultiplier)
    )) {isSetParameters=true}
  ]
  rendererClasses [&NiagaraRibbonRendererProperties]
  renderers [{rendererClass=&NiagaraRibbonRendererProperties}]
 }
 Tip_Trail {
  emitterSpawnScript [&FN_RibbonIDOffsetUtility()]
  emitterUpdateScript [
   &EmitterState()
   &FN_SpawnAnimTrail(AnimTrailWidthMode=::FromFirst)
  ]
  particleSpawnScript [
   &FN_InitializeAnimTrailBoneIndecies(First Bone or Socket Index=1, Second Bone or Socket Index=0) {module=FN_InitializeAnimTrailBoneIndecies001}
   &FN_InitializeAnimTrail(AnimTrailWidth=0.065, Color=<rgba 0.925 0.925 1 0.2>, Lifetime=0.3)
  ]
  particleUpdateScript [
   &SolveForcesAndVelocity(Clamp Velocity=false, Limit Acceleration=false)
   &FN_UpdateAnimTrail() {module=FN_UpdateAnimTrail001}
   &ScaleColor(
    Scale Alpha=&FloatFromCurve(
     FloatCurve=DataInterface() {propertyValues {Curve {keys [
         {arriveTangent=-2.6867, interpMode=RCIM_Cubic, time=0, value=1}
         {arriveTangent=-9.9177, interpMode=RCIM_Cubic, leaveTangent=-9.9177, tangentMode=RCTM_User, time=0.325, value=0}
        ]}}}
     Scale Curve=&FNTimeOfDayScalar(Night Scale=1, Morning Scale=0.8, Evening Scale=0.8)
    )
    Scale RGB=&MakeVectorFromLinearColorRGB(Color=&ColorFromCurve(ColorCurve=DataInterface() {propertyValues {
        AlphaCurve {keys [<at 0.2 1 RCIM_Cubic>,<at 1 0>]}
        BlueCurve {keys [<at 0.1974 1 RCIM_Cubic>,<at 1 0.05 RCIM_Cubic>]}
        GreenCurve {keys [<at 0.1974 1 RCIM_Cubic>,<at 1 0.05 RCIM_Cubic>]}
        RedCurve {keys [<at 0.1974 1 RCIM_Cubic>,<at 1 0.05 RCIM_Cubic>]}
       }}))
    ScaleA=true
    ScaleRGB=true
   )
   &ScaleRibbonWidth(Ribbon Width Scale=&FloatFromCurve(FloatCurve=DataInterface() {propertyValues {Curve {keys [<at 0 1>,<at 1 0>]}}}))
  ]
  rendererClasses [&NiagaraRibbonRendererProperties]
  renderers [{rendererClass=&NiagaraRibbonRendererProperties}]
 }
}
summary {
 systemName=NS_Pickaxe_EvilBunny_Trail
 emitters [
  {emitterName=Main_Trail, rendererClasses [&NiagaraRibbonRendererProperties]}
  {emitterName=Tip_Trail, rendererClasses [&NiagaraRibbonRendererProperties]}
 ]
 userVariables [User.RibbonIDOffset {type=NiagaraInt32}]
}
