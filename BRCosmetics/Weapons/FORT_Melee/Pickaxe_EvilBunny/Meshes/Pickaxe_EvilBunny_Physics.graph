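# Physics asset for the Evil Bunny pickaxe, a two-handed melee weapon cosmetic.
# It has two capsule bodies: a long one along the handle bone and a small one on the dyn_hook_02 bone.
# A single constraint joins the hook to the handle with 120-degree swing limit values and twist locked, which points to a dangling, simulated hook piece.
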PhysicsAsset /BRCosmetics/Weapons/FORT_Melee/Pickaxe_EvilBunny/Meshes/Pickaxe_EvilBunny_Physics
bodies {
 Melee_TwoHanded_Handle [{shapeType=Capsule, length=98.2589, radius=3.4403, center=<0.6788 -1.4454 11.5531>}]
 dyn_hook_02 [{shapeType=Capsule, length=9.7456, radius=5.1101, center=<5.7183 0.1723 -0.0192>, rotation=<rot -0.0151 92.2139 -89.9141>}]
}
constraints [{
  bone1Name=dyn_hook_02
  bone2Name=Melee_TwoHanded_Handle
  swing1LimitDegrees=120
  swing2LimitDegrees=120
  twistLimitDegrees=45
  twistMotion=Locked
 }]
